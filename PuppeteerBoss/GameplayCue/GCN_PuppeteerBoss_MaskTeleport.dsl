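; Burst gameplay cue notify for the Puppeteer boss's mask teleport, judging by its name.
; Its OnBurst event only forwards the target, cue parameters and spawn results to the parent implementation.
; The graph adds no logic of its own, and any spawned effects are not defined in it.
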(fn OnBurst
  (params (Target :type Actor%20Object%20Reference) (Parameters :type Gameplay%20Cue%20Parameters%20Structure%20%28by%20ref%29) (SpawnResults :type Gameplay%20Cue%20Notify%20Spawn%20Result%20Structure%20%28by%20ref%29))
  (|CallParentFunction Target Parameters SpawnResults)
  (return))
